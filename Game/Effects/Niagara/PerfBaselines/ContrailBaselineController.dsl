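; Niagara perf baseline controller that sets up and measures a contrail effect on an animated preview character.
; On test begin, it blocking-loads soft mesh and montage references, avoiding hard-reference memory bloat, and plays the montage as a skydive pose on an added mesh component.
; It spawns the baseline system attached to the owner root, flags it as front-end preview and binds that mesh.
; On end, it prints game- and render-thread stats; tick logic defers to the parent.
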(event EventOnOwnerTick
  (params (DeltaTime :type Float%20%28single-precision%29))
  ;[1 1 1] Tick from the owning actor. Useful if we need to do some delta time dependant update.
  (|CallParentFunction DeltaTime))

(event EventOnBeginTest
  (bind _meshassetref (Variables|Default|GetMeshAssetRef))
  (|CallParentFunction)
  ;[1 1 1] Resolve soft references to mesh and anim assets. Must be soft refs to avoid memory bloat from hard refs being loaded.
  (bind _returnvalue (Utilities|LoadAssetBlocking _meshassetref))
  (bind _asskeletal_mesh (Utilities|Casting|CastToSkeletalMesh _returnvalue)
    (:then
      (bind _returnvalue_2 (Utilities|LoadAssetBlocking (Variables|Default|GetAnimAssetRef)))
      (bind _asanim_montage (Utilities|Casting|CastToAnimMontage _returnvalue_2)
        (:then
          _)
        (:CastFailed
          _))
      ;[1 1 1] Add a preview mesh to attatch to and set it to skydive pose
      (bind _returnvalue_3 (Game|AddComponentbyClass :self (Variables|Baseline|GetOwner) :Class "/Script/Engine.SkeletalMeshComponent"))
      (Components|SkinnedMesh|SetSkinnedAssetandUpdate _returnvalue_3 _asskeletal_mesh true)
      (if (and (Utilities|IsValid _returnvalue_3 :!class /Script/Engine.KismetSystemLibrary) (Utilities|IsValid _asanim_montage :!class /Script/Engine.KismetSystemLibrary))
        (Components|Animation|PlayAnimation _returnvalue_3 _asanim_montage true))
      ;[1 1 1] Spawn the baseline system to measure. Set it's FrontEnd flag so it displays correctly.
      (bind _returnvalue_4 (Baseline|GetSystem))
      (bind _returnvalue_5 (Niagara|SpawnSystemAttached _returnvalue_4 (Variables|/Script/Engine.Actor|GetRootComponent (Variables|Baseline|GetOwner)) "None" "0, 0, 0" "0, 0, 0" "KeepRelativeOffset" false false "None" false))
      (Effects|Components|ParticleSystem|SetBoolParameter _returnvalue_5 "bIsFrontEndPreview" true)
      (Niagara|SetNiagaraSkeletalMeshComponent _returnvalue_5 "SkeletalMesh" _returnvalue_3)
      (Components|Activation|Activate _returnvalue_5 false)
      (bind _returnvalue_6 (Baseline|GetSystem))
      (Development|PrintString 0 (Utilities|String|Append "Generating Perf Baseline for System: " (Utilities|GetDisplayName _returnvalue_6)) true true "(R=0.000000,G=0.660000,B=1.000000,A=1.000000)" (Variables|Baseline|GetTestDuration)))
    (:CastFailed
      (Development|PrintString 0 (Utilities|String|Append "Failed to load skeletal mesh while generating contrail baseline: " (Utilities|ToString(SoftObjectReference) _meshassetref))))))

(event EventOnEndTest
  (params (Stats :type Niagara%20Perf%20Baseline%20Stats%20Structure))
  (bind (_perinstanceavg_gt _perinstanceavg_rt _perinstancemax_gt _perinstancemax_rt) (Utilities|Struct|BreakNiagaraPerfBaselineStats Stats))
  (|CallParentFunction Stats)
  (bind _returnvalue (Baseline|GetSystem))
  (Development|PrintString 0 (Utilities|String|Append "Baseline Results for System: " (Utilities|GetDisplayName _returnvalue) "| GT Avg: " (Utilities|String|ToString(Float) _perinstanceavg_gt) "| GT Max: " (Utilities|String|ToString(Float) _perinstancemax_gt) " | RT Avg: " (Utilities|String|ToString(Float) _perinstanceavg_rt) " | RT Max: " (Utilities|String|ToString(Float) _perinstancemax_rt) " |") true true "(R=0.000000,G=0.660000,B=1.000000,A=1.000000)" 1.0))

(fn OnTickTest
  ;[1 1 1] Return false when we want to finish the baseline gathering.
  ; Parent class will do this when test duration has elapsed but we can place any additional conditions here we need.
  (bind _returnvalue (|CallParentFunction))
  (return _returnvalue))
